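# One-shot Niagara effect for a drum strike on the QuickBurst drums gadget cosmetic.
# Flare makes a brief warm glow with a light, partly dimmed at night by time-of-day scaling.
# NitroFlares throws speed-stretched streaks forward plus two staggered, expanding shockwaves; Sparks adds a burst and quickly decaying stream of falling, SubUV-animated sparks.
# Flare and NitroFlares use spawn groups as visibility tags, letting one emitter drive differently tuned layers across several renderers.
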NiagaraSystem /BRCosmetics/Animation/Props/Gadgets_QuickBurst_Drums/FX/NS_QuickBurst_DrumStrike
dependencies {
 usedDataInterfaces [
  &NiagaraDataInterfaceCurve
  &NiagaraDataInterfaceVector2DCurve
  &NiagaraDataInterfaceSpriteRendererInfo
  &NiagaraDataInterfaceColorCurve
 ]
 usedDynamicInputs []
 usedModules [&SystemState]
}
emitters {
 Flare {
  emitterUpdateScript [
   &EmitterState()
   &SpawnBurst_Instantaneous(Use Spawn Probability=false, UseLoopCountLimit=false)
   &SpawnBurst_Instantaneous(Spawn Group=1, Use Spawn Probability=false, UseLoopCountLimit=false) {module=SpawnBurst_Instantaneous001}
  ]
  particleSpawnScript [
   /SetVariables(Particles.VisibilityTag=&SpawnGroup()) {isSetParameters=true}
   &InitializeParticle(
    Color=&MaskLinearColorBySpawnGroup(Masked Linear Color=<rgba 0.5 0.5 0.5 1>, Passthrough Linear Color=<rgba 1 0.4941 0.124 1>)
    Lifetime=&MaskFloatBySpawnGroup(Masked Float=0.15, Passthrough Float=0.2)
    Position Offset=<1.5 0 0>
    Sprite Size Mode=::NewEnumerator3
    Uniform Sprite Size=&MaskFloatBySpawnGroup(Masked Float=100, Passthrough Float=100)
    UsePositionOffset=true
    Write Material Random=true
   )
   &CameraOffset(Camera Offset Amount=&MaskFloatBySpawnGroup(Masked Float=25, Passthrough Float=30))
   &DynamicMaterialParameters(
    Index 0 Param 1=&MaskFloatBySpawnGroup(Masked Float=1, Passthrough Float=0.75)
    Index 0 Param 2=1
    Index 0 Param 3=&MaskFloatBySpawnGroup(Masked Float=0, Passthrough Float=1)
    Module.Index0Param0Write=true
    Module.Index0Param1Write=true
    Module.Index0Param2Write=true
    Module.Index0Param3Write=true
   )
   /SetVariables(Particles.Corona Dyn Param:Vector4f=<0.25 1 0 0>) {isSetParameters=true}
  ]
  particleUpdateScript [
   &ParticleState()
   &ScaleSpriteSize(
    Scale Sprite Size Mode=::NewEnumerator0
    Uniform Scale Factor=&MaskFloatBySpawnGroup(
     Masked Float=&FloatFromCurve(FloatCurve=DataInterface() {propertyValues {Curve {keys [
          {arriveTangent=1.0813, interpMode=RCIM_Cubic, leaveTangent=1.0813, tangentMode=RCTM_User, time=0, value=0.5}
          {arriveTangent=0.0224, interpMode=RCIM_Cubic, leaveTangent=0.0224, tangentMode=RCTM_User, time=1, value=1}
         ]}}})
     Passthrough Float=&FloatFromCurve(FloatCurve=DataInterface() {propertyValues {Curve {keys [{arriveTangent=2.7513, interpMode=RCIM_Cubic, leaveTangent=2.7513, tangentMode=RCTM_User, time=0, value=0}, <at 1 1>]}}})
    )
   )
   &ScaleColor(
    Scale Alpha=&MaskFloatBySpawnGroup(
     Masked Float=&FNTimeOfDayScalar(Night Scale=0.33, Morning Scale=1, Evening Scale=1)
     Passthrough Float=1
    )
    Scale RGB=&MaskVectorBySpawnGroup(
     Masked Vector=&MakeVectorFromLinearColorRGB(Color=&ColorFromCurve(ColorCurve=DataInterface() {propertyValues {
         AlphaCurve {keys [<at 0 0>,<at 0.2006 1 RCIM_Cubic>,<at 1 0>]}
         BlueCurve {keys [
           <at 0.004 0.0208>
           {arriveTangent=0.365, interpMode=RCIM_Cubic, leaveTangent=0.365, time=0.1965, value=0.7367}
           <at 0.2832 0.1227>
           <at 1 0.0077 RCIM_Cubic>
          ]}
         GreenCurve {keys [
           <at 0.004 0.1206>
           {arriveTangent=2.1208, interpMode=RCIM_Cubic, leaveTangent=2.1208, time=0.1965, value=4.155}
           <at 0.2832 0.7127>
           <at 1 0.0445 RCIM_Cubic>
          ]}
         RedCurve {keys [
           <at 0.004 0.2708>
           {arriveTangent=4.7612, interpMode=RCIM_Cubic, leaveTangent=4.7612, time=0.1965, value=9}
           <at 0.2832 1.6>
           <at 1 0.1 RCIM_Cubic>
          ]}
        }}))
     Passthrough Vector=&Multiply_VectorByFloat(
      Vector=<1.5 1.5 1.5>
      Float=&FloatFromCurve(
       FloatCurve=DataInterface() {propertyValues {Curve {keys [
           {arriveTangent=7.9255, interpMode=RCIM_Cubic, leaveTangent=7.9255, tangentMode=RCTM_User, time=0, value=0}
           {interpMode=RCIM_Cubic, tangentMode=RCTM_User, time=0.2, value=0.95}
           {tangentMode=RCTM_User, time=1, value=0}
          ]}}}
       Scale Curve=0.333
      )
     )
    )
    ScaleA=true
    ScaleRGB=true
   )
   /SetVariables(
    Particles.LightColor=&MultiplyLinearColor(Linear Color=$Particles.Color, Scale Factor=&MakeVector4_FromFloat(Float=20))
    Particles.LightExponent=1
    Particles.LightRadius=40
   ) {isSetParameters=true}
  ]
  rendererClasses [&NiagaraSpriteRendererProperties, &NiagaraLightRendererProperties]
  renderers [
   {rendererClass=&NiagaraSpriteRendererProperties}
   {rendererClass=&NiagaraLightRendererProperties, rendererIndex=1}
   {rendererClass=&NiagaraSpriteRendererProperties, rendererIndex=2}
  ]
 }
 NitroFlares {
  emitterUpdateScript [
   &EmitterState()
   &SpawnBurst_Instantaneous(
    Spawn Count=&RandomRangeInt(Minimum=8, Maximum=22)
    Use Spawn Probability=false
    UseLoopCountLimit=false
   )
   &SpawnBurst_Instantaneous(Spawn Group=1, Use Spawn Probability=false, UseLoopCountLimit=false) {module=SpawnBurst_Instantaneous001}
   &SpawnBurst_Instantaneous(Spawn Group=1, Spawn Time=0.1, Use Spawn Probability=false, UseLoopCountLimit=false) {module=SpawnBurst_Instantaneous002}
  ]
  particleSpawnScript [
   /SetVariables(Particles.VisibilityTag=&SpawnGroup()) {isSetParameters=true}
   &InitializeParticle(
    Color=<rgba 3 2.5508 1.8811 1>
    Lifetime=&MaskFloatBySpawnGroup(Masked Float=&RandomRangeFloat(Minimum=0.05, Maximum=0.15), Passthrough Float=0.3)
    Mesh Scale Mode=::NewEnumerator1
    Mesh Uniform Scale Max=20
    Mesh Uniform Scale Min=10
    Sprite Size Max=<8 15>
    Sprite Size Min=<4 8>
    Sprite Size Mode=::NewEnumerator2
    UsePositionOffset=false
    Write Material Random=true
   )
   &AddVelocity(Velocity=&MaskVectorBySpawnGroup(
     Masked Vector=&RandomRangeVector(Minimum=<100 -200 -200>, Maximum=<300 200 200>)
     Passthrough Vector=<20 0 0>
    ))
   &InitialMeshOrientation(Mesh Axis To Orient=<0 0 1>, Mesh Orientation Mode=::NewEnumerator2, Use Rotation Vector=false)
  ]
  particleUpdateScript [
   &ParticleState()
   &Drag(
    Drag=&MaskFloatBySpawnGroup(Masked Float=&RandomRangeFloat(Minimum=8, Maximum=12), Passthrough Float=0)
    Use Linear Drag=true
    Use Rotational Drag=false
   )
   &SolveForcesAndVelocity(Clamp Velocity=false, Limit Acceleration=false)
   &ScaleMeshSize(Scale Factor=&VectorFromFloat(Value=&MaskFloatBySpawnGroup(
      Masked Float=&FloatFromCurve(FloatCurve=DataInterface() {propertyValues {Curve {keys [{arriveTangent=1.1031, interpMode=RCIM_Cubic, leaveTangent=1.1031, tangentMode=RCTM_User, time=0, value=0.25}, <at 1 1.25 RCIM_Cubic>]}}})
      Passthrough Float=&FloatFromCurve(FloatCurve=DataInterface() {propertyValues {Curve {keys [
           {arriveTangent=5.1547, interpMode=RCIM_Cubic, leaveTangent=5.1547, leaveTangentWeight=0.7454, tangentMode=RCTM_User, time=0, value=1}
           {arriveTangentWeight=0.7454, interpMode=RCIM_Cubic, tangentMode=RCTM_User, time=1, value=3}
          ]}}})
     )))
   &ScaleSpriteSize(Uniform Curve Sprite Scale=DataInterface() {propertyValues {Curve {keys [
        {arriveTangent=6.7711, interpMode=RCIM_Cubic, leaveTangent=6.7711, tangentMode=RCTM_User, time=0, value=0.1}
        {arriveTangent=-0.1, interpMode=RCIM_Cubic, leaveTangent=-0.1, time=0.18, value=0.8}
        <at 1 0 RCIM_Cubic>
       ]}}})
   &ScaleSpriteSizeBySpeed(Max Scale Factor=<0.5 7>, Min Scale Factor=<0.5 1>, Velocity Threshold=666)
   &ScaleColor(
    Scale Alpha=&Multiply_Float(
     A=&FloatFromCurve(
      FloatCurve=DataInterface() {propertyValues {Curve {keys [
          {arriveTangent=9.0725, interpMode=RCIM_Cubic, leaveTangent=9.0725, tangentMode=RCTM_User, time=0, value=1}
          {arriveTangent=-2.5249, interpMode=RCIM_Cubic, leaveTangent=-2.5249, tangentMode=RCTM_User, time=0.1408, value=0.9847}
          {arriveTangent=-0.6199, interpMode=RCIM_Cubic, leaveTangent=-0.6199, time=0.3, value=0.95}
          {arriveTangent=-1.3571, interpMode=RCIM_Cubic, leaveTangent=-1.3571, time=0.6, value=0.7}
          {arriveTangent=-0.0405, interpMode=RCIM_Cubic, leaveTangent=-0.0405, tangentMode=RCTM_User, time=1, value=0}
         ]}}}
      Scale Curve=2
     )
     B=2
    )
    Scale RGB=&VectorFromFloat(Value=&Lerp_Float(
      A=1
      B=2.5
      Alpha=&FloatFromCurve(
       FloatCurve=DataInterface() {propertyValues {Curve {keys [
           {arriveTangent=0.0028, interpMode=RCIM_Cubic, leaveTangent=0.0028, tangentMode=RCTM_User, time=0, value=1}
           {arriveTangent=-2.5249, interpMode=RCIM_Cubic, leaveTangent=-2.5249, tangentMode=RCTM_User, time=0.1408, value=0.9847}
           {arriveTangent=-0.0405, interpMode=RCIM_Cubic, leaveTangent=-0.0405, tangentMode=RCTM_User, time=1, value=0}
          ]}}}
       Scale Curve=1.5
      )
     ))
    ScaleA=true
    ScaleRGB=true
   )
   /SetVariables(
    Particles.ShockwaveColor:LinearColor=&MakeLinearColorFromVectorAndFloat(
     "Vector (RGB)"=&MakeVectorFromLinearColorRGB(Color=<rgba 0.1 0.0696 0.0172 1>)
     "Float (Alpha)"=&FloatFromCurve(FloatCurve=DataInterface() {propertyValues {Curve {keys [
          {interpMode=RCIM_Cubic, leaveTangentWeight=0.4714, tangentMode=RCTM_User, time=0, value=1}
          {arriveTangentWeight=0.4714, interpMode=RCIM_Cubic, tangentMode=RCTM_User, time=1, value=0}
         ]}}})
    )
    Particles.ShockwaveDynParam:Vector4f=&MakeVector4(
     X=0.5
     Y=1
     Z=&RandomRangeFloat(Minimum=-360, Maximum=360)
     W=&FloatFromCurve(FloatCurve=DataInterface() {propertyValues {Curve {keys [
          {arriveTangent=7.0476, interpMode=RCIM_Cubic, leaveTangent=7.0476, leaveTangentWeight=1.0541, tangentMode=RCTM_User, time=0, value=0}
          {arriveTangentWeight=1.0541, interpMode=RCIM_Cubic, tangentMode=RCTM_User, time=1, value=3}
         ]}}})
    )
   ) {isSetParameters=true}
  ]
  rendererClasses [&NiagaraSpriteRendererProperties, &NiagaraMeshRendererProperties]
  renderers [
   {rendererClass=&NiagaraSpriteRendererProperties}
   {rendererClass=&NiagaraMeshRendererProperties, rendererIndex=1}
  ]
 }
 Sparks {
  emitterUpdateScript [
   &EmitterState()
   &SpawnBurst_Instantaneous(
    Spawn Count=&RandomRangeInt(Minimum=15, Maximum=25)
    Use Spawn Probability=false
    UseLoopCountLimit=false
   )
   &SpawnRate(
    SpawnRate=&FloatFromCurve(
     FloatCurve=DataInterface() {propertyValues {Curve {keys [{arriveTangent=-15.6914, interpMode=RCIM_Cubic, leaveTangent=-15.6914, tangentMode=RCTM_User, time=0, value=1}, <at 0.1 0 RCIM_Cubic>]}}}
     CurveIndex=$System.Age
     Scale Curve=300
    )
    Use Spawn Probability=false
   )
  ]
  particleSpawnScript [
   &InitializeParticle(
    Color=<rgba 3 3 3 1>
    Lifetime Max=0.3
    Lifetime Min=0.05
    Lifetime Mode=::NewEnumerator1
    Sprite Size Max=<3 8>
    Sprite Size Min=<1 3>
    Sprite Size Mode=::NewEnumerator2
    UsePositionOffset=false
    Write Material Random=true
   )
   &ShapeLocation(
    Cylinder Height=1
    Cylinder Radius=3
    MaskSpawnGroup=false
    Override Seed=false
    Rotation Mode=::NewEnumerator1
    Shape Primitive=::NewEnumerator1
    Surface Only=true
    Use Endcaps In Surface Only Mode=false
    UseOwnerScale=true
   )
   &AddVelocity(
    Module.Use Falloff Distance=false
    Origin Offset=<-10 0 0>
    Override Seed=false
    Velocity Mode=::NewEnumerator1
    Velocity Speed=&RandomRangeFloat(Minimum=222, Maximum=666)
   )
  ]
  particleUpdateScript [
   &GravityForce(Gravity=<0 0 -666>)
   &Drag(Drag=&RandomRangeFloat(Minimum=4, Maximum=11), Use Linear Drag=true, Use Rotational Drag=false)
   &SolveForcesAndVelocity(Clamp Velocity=false, Limit Acceleration=false)
   &ParticleState()
   &ScaleSpriteSizeBySpeed(Max Scale Factor=<1 3.5>)
   &ScaleSpriteSize(
    Non-Uniform Curve Sprite Scale=DataInterface() {propertyValues {
      XCurve {keys [{tangentMode=RCTM_User, time=0, value=1}, {tangentMode=RCTM_User, time=1, value=1}]}
      YCurve {keys [{tangentMode=RCTM_User, time=0, value=1}, {tangentMode=RCTM_User, time=0.5, value=3}, <at 1 1>]}
     }}
    Scale Sprite Size Mode=::NewEnumerator3
   )
   &SubUVAnimation(
    End Frame Range Override=48
    "Sprite Renderer "=DataInterface() {propertyValues {SpriteRenderer=&self:Sparks_0.NiagaraSpriteRendererProperties_0}}
    Start Frame Range Override=13
    UseEndFrame=true
    UseStartFrame=true
   )
   &ScaleColor(
    Scale Alpha=&FloatFromCurve(FloatCurve=DataInterface() {propertyValues {Curve {keys [<at 0.333 1>,<at 1 0>]}}})
    Scale RGB=&MakeVectorFromLinearColorRGB(Color=&LerpLinearColors(
      StartColor=&ColorFromCurve(ColorCurve=DataInterface() {propertyValues {
         AlphaCurve {keys [<at 0.3609 1>,<at 1 0>]}
         BlueCurve {keys [<at 0 0.6562>,<at 0.333 0>,<at 0.75 0>]}
         GreenCurve {keys [<at 0 1.8281>,<at 0.333 1.5333>,<at 0.75 0.15>]}
         RedCurve {keys [<at 0 3>,<at 0.333 4>,<at 0.75 3>]}
        }})
      EndColor=<rgba 1.24 1.392 2 1>
      LerpFactor=&FloatFromCurve(
       FloatCurve=DataInterface() {propertyValues {Curve {keys [<at 0 1 RCIM_Constant>,<at 0.05 0 RCIM_Constant>]}}}
       CurveIndex=&RandomRangeFloat(Minimum=0, Maximum=1)
      )
     ))
    ScaleA=true
    ScaleRGB=true
   )
   &DynamicMaterialParameters(
    Index 0 Param 1=3
    Index 0 Param 2=20
    Index 0 Param 3=1
    Module.Index0Param0Write=true
    Module.Index0Param1Write=true
    Module.Index0Param2Write=true
    Module.Index0Param3Write=true
   )
  ]
  rendererClasses [&NiagaraSpriteRendererProperties]
  renderers [{rendererClass=&NiagaraSpriteRendererProperties}]
 }
}
summary {
 systemName=NS_QuickBurst_DrumStrike
 emitters [
  {emitterName=Sparks, rendererClasses [&NiagaraSpriteRendererProperties]}
  {emitterName=NitroFlares, rendererClasses [&NiagaraSpriteRendererProperties, &NiagaraMeshRendererProperties]}
  {emitterName=Flare, rendererClasses [&NiagaraSpriteRendererProperties, &NiagaraLightRendererProperties]}
 ]
 userVariables [
  User.bIsFrontEnd {type=NiagaraBool}
  User.bIsFrontEndPreview {type=NiagaraBool}
 ]
}
