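# Physics asset for the QuickBurst Drums gadget prop in BRCosmetics, giving its skeleton one capsule body per listed bone.
# The root has a very large, near-spherical capsule and JNT_Ground_01 an offset, tilted one, while offset, main and both drum joints get only tiny placeholder-sized capsules.
# Constraints follow a root, offset, main chain that branches to the ground and left/right drum joints, all sharing identical swing and twist limits.
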PhysicsAsset /BRCosmetics/Animation/Props/Gadgets_QuickBurst_Drums/Meshes/PA_QuickBurst_Drums_Gadget
bodies {
 JNT_Ground_01 [{shapeType=Capsule, length=79.8692, radius=69.2581, center=<-8.3294 42.5567 41.8001>, rotation=<rot 23.7789 3.388 -90>}]
 JNT_L_Drum [{shapeType=Capsule, length=0.505, radius=0.505, rotation=<rot 0 -90 0>}]
 JNT_R_Drum [{shapeType=Capsule, length=0.505, radius=0.505, rotation=<rot 0 -90 0>}]
 main [{shapeType=Capsule, length=0.505, radius=0.505, rotation=<rot 0 -90 0>}]
 offset [{shapeType=Capsule, length=0.505, radius=0.505, rotation=<rot 0 -90 0>}]
 root [{shapeType=Capsule, length=0.505, radius=365.2789, rotation=<rot 0 -90 0>}]
}
constraints [
 {bone1Name=offset, bone2Name=root, swing1LimitDegrees=45, swing2LimitDegrees=45, twistLimitDegrees=45}
 {bone1Name=main, bone2Name=offset, swing1LimitDegrees=45, swing2LimitDegrees=45, twistLimitDegrees=45}
 {bone1Name=JNT_Ground_01, bone2Name=main, swing1LimitDegrees=45, swing2LimitDegrees=45, twistLimitDegrees=45}
 {bone1Name=JNT_L_Drum, bone2Name=main, swing1LimitDegrees=45, swing2LimitDegrees=45, twistLimitDegrees=45}
 {bone1Name=JNT_R_Drum, bone2Name=main, swing1LimitDegrees=45, swing2LimitDegrees=45, twistLimitDegrees=45}
]
